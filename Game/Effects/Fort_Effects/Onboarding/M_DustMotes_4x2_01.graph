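# Translucent lit particle material for sprite and Niagara dust motes, using volumetric per-vertex directional lighting and a spherical normal map.
# Each particle picks a random frame from a 4x2 dust atlas, tinted by particle color, with boostable emissive from that color.
# Opacity combines atlas and particle alpha with a random level between two parameters.
# An optional flicker blends time-driven sine waves, offset per particle to avoid synchronized pulsing.
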Material /Game/Effects/Fort_Effects/Onboarding/M_DustMotes_4x2_01
graph {
 properties {
  bUsedWithNiagaraSprites=true
  bUsedWithParticleSprites=true
  blendMode=BLEND_Translucent
  pixelDepthOffsetMode=PDOM_Legacy
  translucencyDirectionalLightingIntensity=0.85
  translucencyLightingMode=TLM_VolumetricPerVertexDirectional
  usedShadingModels=Default Lit
 }
 nodes [
  n1 = ParticleColor()
  MaterialFunctionCall(Emissive=Multiply(A=@n1.RGB, B=ScalarParameter() {parameterName=Emissive Boost})) {materialFunction=&MF_FX_EmissiveAdjustments}
  n2 = ParticleRandom()
  n3 = MaterialFunctionCall!(
   Texture=TextureObject() {texture=/Game/Effects/Fort_Effects/Textures/T_DustMotes_4x2_01}
   UVs
   SubImages=Constant2Vector() {g=2, r=4}
   Frame=@n2
  ) {materialFunction=/Engine/Functions/Engine_MaterialFunctions03/Texturing/SubUV_Function}
  Multiply(A=@n1.RGB, B=@n3.RGB)
  n4 = Add(A=Time(), B=ParticleRandom())
  n5 = Clamp(None=LinearInterpolate(A=Sine(None=@n4) {period=1.7421}, B=Sine(None=@n4) {period=2.4587}, Alpha=Cosine(None=@n4) {period=0.8745}))
  n6 = ScalarParameter() {defaultValue=0.175, parameterName=Alpha Max}
  n7 = ScalarParameter() {defaultValue=0.07, parameterName=Alpha MIn}
  Multiply(
   A=LinearInterpolate(A=@n1.A, B=Multiply(A=@n1.A, B=@n5), Alpha=ScalarParameter() {parameterName=Flicker})
   B=Multiply(A=@n3.Alpha, B=LinearInterpolate(A=@n6, B=@n7, Alpha=@n2))
  )
  TextureSample() {samplerType=SAMPLERTYPE_Normal, texture=/Game/Environments/World/Backgrounds/Universal/TechContent/Textures/T_Spherenormal_01}
 ]
}
